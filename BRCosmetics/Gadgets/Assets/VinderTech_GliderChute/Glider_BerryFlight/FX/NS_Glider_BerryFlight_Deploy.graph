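# Niagara system for the deploy effect of the BerryFlight glider: a short burst of layered sprites.
# A glitch sprite (Glith), expanding ring, two glows and radial sparkles that follow source movement fire as bursts, most after a brief delay.
# User.VariantSwitch indexes per-emitter arrays to pick one of four color variants, while front-end and time-of-day functions adjust glow intensity.
# The alternate Glith001 emitter, which lerps between color curves, is disabled.
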NiagaraSystem /BRCosmetics/Gadgets/Assets/VinderTech_GliderChute/Glider_BerryFlight/FX/NS_Glider_BerryFlight_Deploy
dependencies {
 usedDataInterfaces [
  &NiagaraDataInterfaceArrayColor
  &NiagaraDataInterfaceCurve
  &NiagaraDataInterfaceArrayFloat
  &NiagaraDataInterfaceColorCurve
 ]
 usedDynamicInputs []
 usedModules [&SystemState]
}
emitters {
 Glith {
  emitterUpdateScript [
   &EmitterState(Life Cycle Mode=::NewEnumerator1, Loop Behavior=::NewEnumerator1, Loop Duration=3, UseLoopDelay=false)
   &SpawnBurst_Instantaneous(Use Spawn Probability=false, UseLoopCountLimit=false)
  ]
  particleSpawnScript [
   &InitializeParticle(
    Lifetime=0.45
    Sprite Rotation Angle=0.25
    Sprite Rotation Mode=::NewEnumerator1
    Sprite Size Mode=::NewEnumerator3
    Uniform Sprite Size=180
    UsePositionOffset=false
    Write Material Random=true
   )
   &CameraOffset(Camera Offset Amount=50)
  ]
  particleUpdateScript [
   &ParticleState()
   &ScaleSpriteSize(Uniform Curve Sprite Scale=DataInterface() {propertyValues {Curve {keys [
        {arriveTangent=4.2019, interpMode=RCIM_Cubic, leaveTangent=4.2019, tangentMode=RCTM_User, time=0, value=0}
        {arriveTangent=1, interpMode=RCIM_Cubic, leaveTangent=1, time=0.2, value=0.65}
        <at 1 1 RCIM_Cubic>
       ]}}})
   &ScaleColor(
    Scale Alpha=&FloatFromCurve(
     FloatCurve=DataInterface() {propertyValues {Curve {keys [
         <at 0 0 RCIM_Cubic>
         {arriveTangent=-2.4751, interpMode=RCIM_Cubic, leaveTangent=-2.4751, tangentMode=RCTM_User, time=0.2, value=1}
         <at 1 0 RCIM_Cubic>
        ]}}}
     Scale Curve=&SelectFloatFromArray(
      Direct Array Index=&MakeIntFromFloat(Float=$User.VariantSwitch)
      Float Selection Array=DataInterface() {propertyValues {FloatData [1,1,2,15]}}
     )
    )
    Scale RGB=&MakeVectorFromLinearColorRGB(Color=&SelectLinearColorFromArray(
      Array Sampling Mode=::NewEnumerator1
      Direct Array Index=&MakeIntFromFloat(Float=$User.VariantSwitch)
      Color Selection Array=DataInterface() {propertyValues {ColorData [<rgba 10 1.5828 2.9259 1>, <rgba 6.7122 1.5165 20 1>, <rgba 10 2.5249 0.9017 1>, <rgba 0.6595 2.416 15 1>]}}
      Random Seed=?
     ))
    ScaleA=true
    ScaleRGB=true
   )
  ]
  rendererClasses [&NiagaraSpriteRendererProperties]
  renderers [{rendererClass=&NiagaraSpriteRendererProperties}]
 }
 Glith001 {
  bEnabled=false
  emitterUpdateScript [
   &EmitterState(Life Cycle Mode=::NewEnumerator1, Loop Behavior=::NewEnumerator1, Loop Duration=3, UseLoopDelay=false)
   &SpawnBurst_Instantaneous(Use Spawn Probability=false, UseLoopCountLimit=false)
  ]
  particleSpawnScript [
   &InitializeParticle(
    Lifetime=0.5
    Sprite Rotation Angle=0.25
    Sprite Rotation Mode=::NewEnumerator1
    Sprite Size Mode=::NewEnumerator3
    Uniform Sprite Size=180
    UsePositionOffset=false
    Write Material Random=true
   )
   &CameraOffset(Camera Offset Amount=50)
  ]
  particleUpdateScript [
   &ParticleState()
   &ScaleSpriteSize(Uniform Curve Sprite Scale=DataInterface() {propertyValues {Curve {keys [
        {arriveTangent=4.2019, interpMode=RCIM_Cubic, leaveTangent=4.2019, tangentMode=RCTM_User, time=0, value=0}
        {arriveTangent=1, interpMode=RCIM_Cubic, leaveTangent=1, time=0.2, value=0.65}
        <at 1 1 RCIM_Cubic>
       ]}}})
   &ScaleColor(
    Scale Alpha=&FloatFromCurve(
     FloatCurve=DataInterface() {propertyValues {Curve {keys [
         <at 0 0 RCIM_Cubic>
         {arriveTangent=-2.4751, interpMode=RCIM_Cubic, leaveTangent=-2.4751, tangentMode=RCTM_User, time=0.2, value=1}
         <at 1 0 RCIM_Cubic>
        ]}}}
     Scale Curve=&SelectFloatFromArray(
      Direct Array Index=&MakeIntFromFloat(Float=$User.VariantSwitch)
      Float Selection Array=DataInterface() {propertyValues {FloatData [1,300,300,300]}}
     )
    )
    Scale RGB=/Niagara/DynamicInputs/Math/LerpMultipleVectors(
     Alpha:NiagaraFloat=$User.VariantSwitch
     V 1:Vector3f=&MakeVectorFromLinearColorRGB(Color=&ColorFromCurve(ColorCurve=DataInterface() {propertyValues {
         AlphaCurve {keys []}
         BlueCurve {keys [<at 0.05 0.2664 RCIM_Cubic>,<at 0.25 0.5053>,<at 0.57 0.1032 RCIM_Cubic>]}
         GreenCurve {keys [<at 0.05 0.1515 RCIM_Cubic>,<at 0.25 0.0972>,<at 0.57 0.0267 RCIM_Cubic>]}
         RedCurve {keys [<at 0.05 1 RCIM_Cubic>,<at 0.25 20>,<at 0.57 1 RCIM_Cubic>]}
        }}))
     V 2:Vector3f=&MakeVectorFromLinearColorRGB(Color=&ColorFromCurve(ColorCurve=DataInterface() {propertyValues {
         AlphaCurve {keys []}
         BlueCurve {keys [<at 0.0838 1 RCIM_Cubic>,<at 0.25 20>,<at 0.522 1 RCIM_Cubic>]}
         GreenCurve {keys [<at 0.0838 0.1135 RCIM_Cubic>,<at 0.25 0.4>,<at 0.522 0.0797 RCIM_Cubic>]}
         RedCurve {keys [<at 0.0838 0.3934 RCIM_Cubic>,<at 0.25 8.0822>,<at 0.522 0.1061 RCIM_Cubic>]}
        }}))
     V 3:Vector3f=&MakeVectorFromLinearColorRGB(Color=&ColorFromCurve(ColorCurve=DataInterface() {propertyValues {
         AlphaCurve {keys []}
         BlueCurve {keys [<at 0.073 1>,<at 0.25 0.1228>,<at 0.5128 0.6354>]}
         GreenCurve {keys [<at 0.073 0.3013>,<at 0.25 3.689>,<at 0.5128 0.1732>]}
         RedCurve {keys [<at 0.073 0.095>,<at 0.25 20>,<at 0.5128 0.0421>]}
        }}))
     V 4:Vector3f=&MakeVectorFromLinearColorRGB(Color=&ColorFromCurve(ColorCurve=DataInterface() {propertyValues {
         AlphaCurve {keys []}
         BlueCurve {keys [<at 0.0746 1>,<at 0.25 20>,<at 0.522 1 RCIM_Cubic>]}
         GreenCurve {keys [<at 0.0746 0.2207>,<at 0.25 4.4145>,<at 0.522 0.0797 RCIM_Cubic>]}
         RedCurve {keys [<at 0.0746 0>,<at 0.25 0>,<at 0.522 0.1061 RCIM_Cubic>]}
        }}))
     V 5:Vector3f=?
     Vector Count:ENiagara_MultipleLerpCount=::NewEnumerator1
    )
    ScaleA=true
    ScaleRGB=true
   )
  ]
  rendererClasses [&NiagaraSpriteRendererProperties]
  renderers [{rendererClass=&NiagaraSpriteRendererProperties}]
 }
 Ring {
  emitterUpdateScript [
   &EmitterState(Life Cycle Mode=::NewEnumerator1, Loop Behavior=::NewEnumerator1, Loop Delay=0.05, Loop Duration=1, UseLoopDelay=true)
   &SpawnBurst_Instantaneous(Use Spawn Probability=false, UseLoopCountLimit=false)
  ]
  particleSpawnScript [
   &InitializeParticle(
    Color Mode=::NewEnumerator0
    Lifetime=0.4
    Sprite Size Mode=::NewEnumerator3
    Uniform Sprite Size=105
    UsePositionOffset=false
    Write Material Random=true
   )
   &CameraOffset(Camera Offset Amount=50)
   &DynamicMaterialParameters(
    Index 0 Param 1=0.6
    Index 0 Param 2=1
    Index 0 Param 3=1
    Index 0 Param 4=1
    Module.Index0Float0Write=true
    Module.Index0Float1Write=true
    Module.Index0Float2Write=true
    Module.Index0Float3Write=true
   )
  ]
  particleUpdateScript [
   &ParticleState()
   &ScaleSpriteSize(Scale Factor=&Vector2DFromFloat(Value=&FloatFromCurve(FloatCurve=DataInterface() {propertyValues {Curve {keys [{arriveTangent=5.3461, interpMode=RCIM_Cubic, leaveTangent=5.3461, tangentMode=RCTM_User, time=0, value=0}, <at 1 2 RCIM_Cubic>]}}})))
   &ScaleColor(
    Scale Alpha=&FloatFromCurve(
     FloatCurve=DataInterface() {propertyValues {Curve {keys [
         {arriveTangent=6.4019, interpMode=RCIM_Cubic, leaveTangent=6.4019, tangentMode=RCTM_User, time=0, value=0}
         {arriveTangent=0.002, interpMode=RCIM_Cubic, leaveTangent=0.002, time=0.2, value=0.7}
         {arriveTangent=-0.9006, interpMode=RCIM_Cubic, leaveTangent=-0.9006, tangentMode=RCTM_User, time=1.0011, value=0.002}
        ]}}}
     Scale Curve=3
    )
    Scale RGB=&MakeVectorFromLinearColorRGB(Color=&SelectLinearColorFromArray(
      Array Sampling Mode=::NewEnumerator1
      Direct Array Index=&MakeIntFromFloat(Float=$User.VariantSwitch)
      Color Selection Array=DataInterface() {propertyValues {ColorData [<rgba 1 0.1583 0.2926 1>, <rgba 0.3356 0.0758 1 1>, <rgba 1 0.2525 0.0902 1>, <rgba 0.0528 0.1933 1.2 1>]}}
      Random Seed=?
     ))
    ScaleA=true
    ScaleRGB=true
   )
  ]
  rendererClasses [&NiagaraSpriteRendererProperties]
  renderers [{rendererClass=&NiagaraSpriteRendererProperties}]
 }
 Trail_Sparkles {
  emitterUpdateScript [
   &EmitterState(Life Cycle Mode=::NewEnumerator1, Loop Behavior=::NewEnumerator1, Loop Delay=0.05, Loop Duration=3, UseLoopDelay=true)
   &SpawnBurst_Instantaneous(Spawn Count=25, Use Spawn Probability=false, UseLoopCountLimit=false)
  ]
  particleSpawnScript [
   &PartitionParticles()
   &InitializeParticle(
    Lifetime Max=0.6
    Lifetime Min=0.2
    Lifetime Mode=::NewEnumerator1
    Sprite Size Mode=::NewEnumerator1
    Uniform Sprite Size Max=20
    Uniform Sprite Size Min=10
    UsePositionOffset=false
    Write Material Random=true
   )
   &ShapeLocation(MaskSpawnGroup=false, Override Seed=false, Sphere Radius=30, UseOwnerScale=true)
   &AddVelocity(
    Rotation Coordinate Space=::Simulation
    Velocity=&NormalizeVector(Vector To Normalize=$Particles.ShapeLocation.ShapeVector, Safe Normalize=true, Safe Fallback Threshold=0, Safe Fallback Vector=<0 0 0.001>)
    Velocity Speed Scale=&RandomRangeFloat(Minimum=200, Maximum=300)
   )
  ]
  particleUpdateScript [
   &ParticleState()
   &ScaleSpriteSize(Scale Factor=&Vector2DFromFloat(Value=&FloatFromCurve(FloatCurve=DataInterface() {propertyValues {Curve {keys [
          {arriveTangent=6.2488, interpMode=RCIM_Cubic, leaveTangent=6.2488, tangentMode=RCTM_User, time=0, value=0.25}
          {arriveTangent=0.059, interpMode=RCIM_Cubic, leaveTangent=0.059, tangentMode=RCTM_User, time=0.2, value=1}
          {arriveTangent=0.0499, interpMode=RCIM_Cubic, leaveTangent=0.0499, tangentMode=RCTM_User, time=0.6, value=1}
          {arriveTangent=-0.4388, interpMode=RCIM_Cubic, leaveTangent=-0.4388, tangentMode=RCTM_User, time=1, value=0}
         ]}}})))
   &ScaleColor(
    Scale Alpha=&FloatFromCurve(
     FloatCurve=DataInterface() {propertyValues {Curve {
        defaultValue=0
        keys [
         {arriveTangent=10.1702, interpMode=RCIM_Cubic, leaveTangent=10.1702, tangentMode=RCTM_User, time=0, value=0}
         {interpMode=RCIM_Cubic, tangentMode=RCTM_User, time=0.2, value=1}
         {arriveTangent=-1.25, interpMode=RCIM_Cubic, leaveTangent=-1.25, time=0.5, value=0.325}
         {tangentMode=RCTM_User, time=1, value=0}
        ]
       }}}
     Scale Curve=2
    )
    Scale RGB=&Multiply_Vector(
     A=&MakeVectorFromLinearColorRGB(Color=&SelectLinearColorFromArray(
       Array Sampling Mode=::NewEnumerator1
       Direct Array Index=&MakeIntFromFloat(Float=$User.VariantSwitch)
       Color Selection Array=DataInterface() {propertyValues {ColorData [<rgba 1 0.1583 0.2926 1>, <rgba 0.353 0.1 1 1>, <rgba 0.8 0.2019 0.072 1>, <rgba 0.044 0.1611 1 1>]}}
       Random Seed=?
      ))
     B=&VectorFromFloat(Value=&FloatFromCurve(FloatCurve=DataInterface() {propertyValues {Curve {keys [<at 0 5>,<at 1 3.5>]}}}))
    )
    ScaleA=true
    ScaleRGB=true
   )
   /SetVariables(Particles.Glow03_Color=&MakeLinearColorFromVectorAndFloat(
     "Vector (RGB)"=&MakeVectorFromLinearColorRGB(Color=$Particles.Color)
     "Float (Alpha)"=&Multiply_Float(
      A=&FNTimeOfDayScalar(Night Scale=1, Morning Scale=0.3, Evening Scale=0.3)
      B=&FloatFromCurve(FloatCurve=DataInterface() {propertyValues {Curve {keys [<at 0 1>,<at 0.1 1>,<at 0.7 1>,<at 1 0>]}}}, Scale Curve=0.2)
     )
    )) {isSetParameters=true}
   /SetVariables(Particles.Glow03_Size=&Multiply_Vector2D_ByFloat(
     Float=&Lerp_Float(
      A=&FloatFromCurve(
       FloatCurve=DataInterface() {propertyValues {Curve {keys [
           <at 0 0 RCIM_Cubic>
           <at 0.2 1.3 RCIM_Cubic>
           {arriveTangent=-0.1423, interpMode=RCIM_Cubic, leaveTangent=-0.1423, tangentMode=RCTM_User, time=1, value=0}
          ]}}}
       Scale Curve=20
      )
      B=&FloatFromCurve(
       FloatCurve=DataInterface() {propertyValues {Curve {keys [
           <at 0 0 RCIM_Cubic>
           {arriveTangent=-0.0069, interpMode=RCIM_Cubic, leaveTangent=-0.0069, tangentMode=RCTM_User, time=0.1, value=1.2}
           {arriveTangent=-0.6113, interpMode=RCIM_Cubic, leaveTangent=-0.6113, tangentMode=RCTM_User, time=0.7, value=1}
           {arriveTangent=-0.1423, interpMode=RCIM_Cubic, leaveTangent=-0.1423, tangentMode=RCTM_User, time=0.8, value=0}
          ]}}}
       Scale Curve=20
      )
      Alpha=$System.InterpolateOverTime.InterpolateOverTime_Float.Moving Average
     )
     Vector2D=<1 1>
    )) {isSetParameters=true}
   &Drag(Drag=2, Use Linear Drag=true, Use Rotational Drag=true)
   &InheritSourceMovement(ApplyForce=false, ApplyPosition=true, ApplyPositionFalloff=false)
   &SolveForcesAndVelocity(Clamp Velocity=false, Limit Acceleration=false)
   /SetVariables(Particles.LocalPosition:Vector3f=&Add_Vector(
     A=Linked() {linkedVariable {name=Particles.Position, type=NiagaraPosition}}
     B=Linked() {linkedVariable {name=Engine.Owner.Position, type=NiagaraPosition}}
    )) {isSetParameters=true}
  ]
  rendererClasses [&NiagaraSpriteRendererProperties]
  renderers [
   {rendererClass=&NiagaraSpriteRendererProperties}
   {rendererClass=&NiagaraSpriteRendererProperties, rendererIndex=1}
  ]
 }
 glow {
  emitterUpdateScript [
   &EmitterState(Life Cycle Mode=::NewEnumerator1, Loop Behavior=::NewEnumerator1, Loop Delay=0.05, Loop Duration=1, UseLoopDelay=true)
   &SpawnBurst_Instantaneous(Spawn Count=1, Use Spawn Probability=false)
  ]
  particleSpawnScript [
   &InitializeParticle(
    Color=&SelectLinearColorFromArray(
     Array Sampling Mode=::NewEnumerator1
     Direct Array Index=&MakeIntFromFloat(Float=$User.VariantSwitch)
     Color Selection Array=DataInterface() {propertyValues {ColorData [<rgba 1 0.1592 0.2036 0.5>, <rgba 0.3246 0.0735 1 0.5>, <rgba 1 0.2676 0.0644 0.5>, <rgba 0.0704 0.0984 1 1>]}}
     Random Seed=?
    )
    Lifetime=0.4
    Sprite Rotation Mode=::NewEnumerator2
    Sprite Size Mode=::NewEnumerator3
    Uniform Sprite Size=500
    UsePositionOffset=false
    Write Material Random=true
   ) {module=InitializeParticle001}
   &CameraOffset(Camera Offset Amount=50)
  ]
  particleUpdateScript [
   &ParticleState()
   &ScaleSpriteSize(Scale Factor=&Vector2DFromFloat(Value=&FloatFromCurve(FloatCurve=DataInterface() {propertyValues {Curve {keys [{arriveTangent=5.3461, interpMode=RCIM_Cubic, leaveTangent=5.3461, tangentMode=RCTM_User, time=0, value=0}, <at 1 2 RCIM_Cubic>]}}})))
   &DynamicMaterialParameters(
    Index 0 Param 1=0.5
    Index 0 Param 2=1
    Module.Param0WriteEnabled=true
    Module.Param1WriteEnabled=true
    Module.Param2WriteEnabled=true
    Module.Param3WriteEnabled=true
   )
   &ScaleColor(
    Scale Alpha=&FloatFromCurve(
     FloatCurve=DataInterface() {propertyValues {Curve {keys [{interpMode=RCIM_Cubic, tangentMode=RCTM_User, time=0, value=2}, {interpMode=RCIM_Cubic, tangentMode=RCTM_User, time=1, value=0}]}}}
     Scale Curve=&FN_FrontEndFloatModifier(Gameplay=1, Front End=0.5)
    )
    ScaleA=true
    ScaleRGB=true
   )
  ]
  rendererClasses [&NiagaraSpriteRendererProperties]
  renderers [{rendererClass=&NiagaraSpriteRendererProperties}]
 }
 glow001 {
  emitterUpdateScript [
   &EmitterState(Life Cycle Mode=::NewEnumerator1, Loop Behavior=::NewEnumerator1, Loop Delay=0.05, Loop Duration=1, UseLoopDelay=true)
   &SpawnBurst_Instantaneous(Spawn Count=1, Use Spawn Probability=false)
  ]
  particleSpawnScript [
   &InitializeParticle(
    Color=&SelectLinearColorFromArray(
     Array Sampling Mode=::NewEnumerator1
     Direct Array Index=&MakeIntFromFloat(Float=$User.VariantSwitch)
     Color Selection Array=DataInterface() {propertyValues {ColorData [<rgba 1 0.1566 0.3006 1>, <rgba 0.28 0.0632 1 1>, <rgba 1 0.2486 0.061 1>, <rgba 0.0146 0.1398 1 1>]}}
     Random Seed=?
    )
    Lifetime=0.3
    Position Offset=<0 2 0>
    Sprite Size=<150 100>
    Sprite Size Mode=::NewEnumerator4
    UsePositionOffset=true
    Write Material Random=true
   ) {module=InitializeParticle001}
   &CameraOffset(Camera Offset Amount=&MakeCustomFloatFromBool(Bool=$User.bIsFrontEndPreview, True Float=0, False Float=50))
  ]
  particleUpdateScript [
   &ParticleState()
   &ScaleSpriteSize(Scale Factor=&Vector2DFromFloat(Value=&FloatFromCurve(FloatCurve=DataInterface() {propertyValues {Curve {keys [{arriveTangent=6.388, interpMode=RCIM_Cubic, leaveTangent=6.388, tangentMode=RCTM_User, time=0, value=1.125}, <at 1 0.875 RCIM_Cubic>]}}})))
   &DynamicMaterialParameters(
    Index 0 Param 1=0.8
    Index 0 Param 2=1
    Index 0 Param 3=1
    Module.Param0WriteEnabled=true
    Module.Param1WriteEnabled=true
    Module.Param2WriteEnabled=true
    Module.Param3WriteEnabled=true
   )
   &ScaleColor(
    Scale Alpha=&FloatFromCurve(
     FloatCurve=DataInterface() {propertyValues {Curve {keys [
         {arriveTangent=-3.2132, interpMode=RCIM_Cubic, leaveTangent=-3.2132, tangentMode=RCTM_User, time=0, value=2}
         {interpMode=RCIM_Cubic, tangentMode=RCTM_User, time=1, value=0}
        ]}}}
     Scale Curve=&FN_FrontEndFloatModifier(Gameplay=1, Front End=0.5)
    )
    Scale RGB=&VectorFromFloat(Value=&Multiply_Float(A=2, B=&FNTimeOfDayScalar(Night Scale=0.5, Morning Scale=1, Evening Scale=1)))
    ScaleA=true
    ScaleRGB=true
   )
  ]
  rendererClasses [&NiagaraSpriteRendererProperties]
  renderers [{rendererClass=&NiagaraSpriteRendererProperties}]
 }
}
summary {
 systemName=NS_Glider_BerryFlight_Deploy
 emitters [
  {emitterName=glow, rendererClasses [&NiagaraSpriteRendererProperties]}
  {emitterName=Ring, rendererClasses [&NiagaraSpriteRendererProperties]}
  {emitterName=Trail_Sparkles, rendererClasses [&NiagaraSpriteRendererProperties]}
  {emitterName=glow001, rendererClasses [&NiagaraSpriteRendererProperties]}
  {emitterName=Glith, rendererClasses [&NiagaraSpriteRendererProperties]}
  {bEnabled=false, emitterName=Glith001, rendererClasses [&NiagaraSpriteRendererProperties]}
 ]
 userVariables [
  User.bIsFrontEnd {type=NiagaraBool}
  User.bIsFrontEndPreview {type=NiagaraBool}
  User.VariantSwitch {type=NiagaraFloat, defaultValue=0, description=""}
 ]
}
