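# Niagara system for the BerryFlight glider's idle effects: ribbon trails, glitch meshes, speed lines and sparkles.
# Paired anim-trail ribbons follow bones or sockets, while partitioned GPU glitch and sparkle particles sample the skeletal mesh.
# System-smoothed floats and User.ForwardDot drive spawning, width, velocity and opacity, with stepped looped-age curves adding flicker.
# VariantSwitch selects palettes and intensities, and bIsFrontEndPreview substitutes preview-specific values.
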NiagaraSystem /BRCosmetics/Gadgets/Assets/VinderTech_GliderChute/Glider_BerryFlight/FX/NS_Glider_BerryFlight_Idle
dependencies {
 usedDataInterfaces [
  &NiagaraDataInterfaceCurve
  &NiagaraDataInterfaceArrayFloat
  &NiagaraDataInterfaceArrayColor
  &NiagaraDataInterfaceArrayInt32
  &NiagaraDataInterfaceArrayFloat3
  &NiagaraDataInterfaceColorCurve
  &NiagaraDataInterfaceVector2DCurve
  &NiagaraDataInterfaceSocketReader
  &NiagaraDataInterfaceSkeletalMesh
 ]
 usedDynamicInputs []
 usedModules [
  &self:SystemSpawnScript.SystemScriptSource.SystemScriptGraph.NiagaraNodeAssignment_0.SetVariables_#1
  &self:SystemSpawnScript.SystemScriptSource.SystemScriptGraph.NiagaraNodeAssignment_1.SetVariables_#2
  &SystemState
  &InterpolateOverTime
 ]
}
emitters {
 Glitch {
  simTarget=GPUComputeSim
  emitterUpdateScript [
   &EmitterState(Delay First Loop Only=true, Life Cycle Mode=::NewEnumerator1, Loop Delay=2, Loop Duration=3, UseLoopDelay=true)
   &SpawnRate(
    Spawn Probability=&MakeCustomFloatFromBool(
     Bool=$User.bIsFrontEndPreview
     True Float=1
     False Float=&Lerp_Float(A=-2, B=1, Alpha=$System.InterpolateOverTime.InterpolateOverTime_Float.Moving Average)
    )
    SpawnRate=300
    Use Spawn Probability=true
   )
  ]
  particleSpawnScript [
   &PartitionParticles()
   &InitializeParticle(
    Color=&RandomRangeLinearColor(Minimum=<rgba 0.0998 0.9809 1 1.5>, Maximum=<rgba 0.4079 0.1713 0.5 1.5>)
    Lifetime Max=0.3
    Lifetime Min=0.3
    Lifetime Mode=::NewEnumerator1
    Mesh Scale Max=&MakeCustomVectorFromBool(Bool=&RandomBool(), True Vector=<1 1 4>, False Vector=<2 2 10>)
    Mesh Scale Min=<0.5 0.5 5>
    Mesh Scale Mode=::NewEnumerator2
    UsePositionOffset=true
    Write Material Random=true
   )
   &ShapeLocation(
    Cylinder Height=30
    Cylinder Radius=1
    MaskSpawnGroup=false
    Offset Mode=::NewEnumerator0
    Override Seed=false
    Rotation Mode=::NewEnumerator1
    Shape Origin=$Particles.Position
    Shape Primitive=::NewEnumerator1
    Surface Only=false
    UseOwnerScale=true
   )
   &InitialMeshOrientation(Mesh Orientation Mode=::NewEnumerator2, Rotation=<0 0.25 0>, Use Rotation Vector=true)
   /SetVariables(Particles.VisibilityTag=&MakeCustomIntFromBool(
     Bool=&RandomBool()
     True INT=&RandomRangeInt(Minimum=0, Maximum=1)
     False INT=1
    )) {isSetParameters=true}
  ]
  particleUpdateScript [
   &ParticleState()
   &SkeletalMeshLocation(
    Bone / Socket Index=&SelectIntFromArray(Direct Array Index=$Particles.PartitionParticles.Partition, Int Selection Array=DataInterface() {propertyValues {IntData [0,1]}})
    Bone Sampling Mode=::NewEnumerator4
    MaskSpawnGroup=false
    Position Sampling=::NewEnumerator1
    Sampled Position Offset=&AddVectorToPosition(
     Position=Linked() {linkedVariable {name=Particles.Initial.ShapeLocation.ShapeVector, type=Vector3f}}
     Vector=&MakeCustomVectorFromBool(
      Bool=$User.bIsFrontEndPreview
      True Vector=&Lerp_Vector(
       A=&SelectVectorFromArray(
        Direct Array Index=$Particles.PartitionParticles.Partition
        Vector Selection Array=DataInterface() {propertyValues {FloatData [<0 5 0>, <0 5 0>]}}
       )
       B=&SelectVectorFromArray(
        Direct Array Index=$Particles.PartitionParticles.Partition
        Vector Selection Array=DataInterface() {propertyValues {FloatData [<0 300 130>, <0 300 130>]}}
       )
       Alpha=&FloatFromCurve(FloatCurve=DataInterface() {propertyValues {Curve {keys [<at 0 0>,<at 1 1>]}}})
      )
      False Vector=&Lerp_Vector(
       A=&SelectVectorFromArray(
        Direct Array Index=$Particles.PartitionParticles.Partition
        Vector Selection Array=DataInterface() {propertyValues {FloatData [<0 5 0>, <0 5 0>]}}
       )
       B=&SelectVectorFromArray(
        Direct Array Index=$Particles.PartitionParticles.Partition
        Vector Selection Array=DataInterface() {propertyValues {FloatData [<0 300 550>, <0 300 550>]}}
       )
       Alpha=&FloatFromCurve(FloatCurve=DataInterface() {propertyValues {Curve {keys [<at 0 0>,<at 1 1>]}}})
      )
     )
    )
    Sampled Position Offset Coordinate Space:ENiagaraSkelMeshTransforms=::NewEnumerator2
    Skeletal Mesh=$System.Skeletal Mesh Reader
   )
   &ScaleMeshSize(Scale Factor=&Multiply_VectorByFloat(
     Vector=<0.7 0.7 1>
     Float=&Multiply_Float(
      A=&FloatFromCurve(
       FloatCurve=DataInterface() {propertyValues {Curve {keys [<at 0 1 RCIM_Constant>,<at 0.1 0 RCIM_Constant>,<at 0.4 1 RCIM_Constant>,<at 0.8 0 RCIM_Constant>,<at 1.2 1 RCIM_Constant>,<at 2 0 RCIM_Constant>,<at 3 1 RCIM_Constant>]}}}
       CurveIndex=$Emitter.LoopedAge
      )
      B=&FloatFromCurve(
       FloatCurve=DataInterface() {propertyValues {Curve {keys [
           {arriveTangent=4.1785, time=0, value=1}
           <at 0.5 1.1>
           {arriveTangent=-0.0405, interpMode=RCIM_Cubic, leaveTangent=-0.0405, tangentMode=RCTM_User, time=1, value=1.5}
          ]}}}
       Scale Curve=0.11
      )
     )
    ))
   &Drag(Drag=2, Use Linear Drag=true, Use Rotational Drag=true)
   &SolveForcesAndVelocity(Clamp Velocity=false, Limit Acceleration=false)
   &DynamicMaterialParameters(Module.Param0WriteEnabled=true, Module.Param1WriteEnabled=true, Module.Param2WriteEnabled=true, Module.Param3WriteEnabled=true)
   &ScaleColor(
    Scale Alpha=&Multiply_Float(
     A=&FloatFromCurve(
      FloatCurve=DataInterface() {propertyValues {Curve {keys [
          {arriveTangent=9.0725, interpMode=RCIM_Cubic, leaveTangent=9.0725, tangentMode=RCTM_User, time=0, value=0}
          {arriveTangent=-2.5249, interpMode=RCIM_Cubic, leaveTangent=-2.5249, tangentMode=RCTM_User, time=0.2, value=0.9847}
          {arriveTangent=-0.0405, interpMode=RCIM_Cubic, leaveTangent=-0.0405, tangentMode=RCTM_User, time=1, value=0}
         ]}}}
      Scale Curve=&RandomRangeFloat(Minimum=0.2, Maximum=1)
     )
     B=&FloatFromCurve(
      FloatCurve=DataInterface() {propertyValues {Curve {keys [<at 0 1 RCIM_Constant>,<at 0.05 0.2 RCIM_Constant>,<at 0.12 1 RCIM_Constant>,<at 0.2 0.2 RCIM_Constant>,<at 0.35 1 RCIM_Constant>,<at 0.5 0.2 RCIM_Constant>,<at 0.7 0.2 RCIM_Constant>,<at 0.8 1 RCIM_Constant>,<at 1 0.2 RCIM_Constant>]}}}
      CurveIndex=$Emitter.LoopedAge
     )
    )
    Scale RGB=&Multiply_VectorByFloat(
     Vector=&MakeVectorFromLinearColorRGB(Color=&ColorFromCurve(ColorCurve=DataInterface() {propertyValues {
         AlphaCurve {keys []}
         BlueCurve {keys [
           <at 0 5 RCIM_Cubic>
           <at 0.1326 0>
           <at 0.2869 1.408>
           <at 0.4134 5>
           <at 0.5108 0.2805>
           <at 0.6236 5>
           {arriveTangent=-25.2941, interpMode=RCIM_Cubic, leaveTangent=-25.2941, time=0.7107, value=0}
           <at 0.8097 0.2928>
          ]}
         GreenCurve {keys [
           <at 0 2.5535 RCIM_Cubic>
           <at 0.1326 0>
           <at 0.2869 5>
           <at 0.4134 0>
           <at 0.5108 0>
           <at 0.6236 4.8554>
           {arriveTangent=-26.0903, interpMode=RCIM_Cubic, leaveTangent=-26.0903, time=0.7107, value=0}
           <at 0.8097 0>
          ]}
         RedCurve {keys [
           <at 0 0 RCIM_Cubic>
           <at 0.1326 1>
           <at 0.2869 2.2363>
           <at 0.4134 0.4394>
           <at 0.5108 5>
           <at 0.6236 0>
           {arriveTangent=26.8676, interpMode=RCIM_Cubic, leaveTangent=26.8676, time=0.7107, value=1}
           <at 0.8097 5>
          ]}
        }}))
     Float=&FloatFromCurve(
      FloatCurve=DataInterface() {propertyValues {Curve {keys [{arriveTangent=-2.0684, interpMode=RCIM_Cubic, leaveTangent=-2.0684, tangentMode=RCTM_User, time=0, value=2}, <at 1 0>]}}}
      Scale Curve=&Multiply_Float(A=0.5, B=1)
     )
    )
    ScaleA=true
    ScaleRGB=true
   )
   &FN_Mesh_Attributes(
    Color=&MakeLinearColorFromVectorAndFloat(
     "Vector (RGB)"=&Multiply_Vector(
      A=&MakeVectorFromLinearColorRGB(Color=<rgba 0.4596 0.9809 1 10>)
      B=&Multiply_VectorByFloat(
       Vector=&MakeVectorFromLinearColorRGB(Color=&ColorFromCurve(ColorCurve=DataInterface() {propertyValues {
           AlphaCurve {keys []}
           BlueCurve {keys [
             <at 0 5 RCIM_Cubic>
             <at 0.1326 0>
             <at 0.2869 1.4885>
             <at 0.4134 5>
             <at 0.5507 0.2805>
             <at 0.5846 5>
             <at 0.8175 5>
             {arriveTangent=-29.7003, interpMode=RCIM_Cubic, leaveTangent=-29.7003, time=0.8628, value=0}
             <at 0.976 0.2928>
            ]}
           GreenCurve {keys [
             <at 0 2.5535 RCIM_Cubic>
             <at 0.1326 0>
             <at 0.2869 5>
             <at 0.4134 0>
             <at 0.5507 0>
             <at 0.5846 4.8554>
             <at 0.8175 4.8554>
             {arriveTangent=-30.6352, interpMode=RCIM_Cubic, leaveTangent=-30.6352, time=0.8628, value=0}
             <at 0.976 0>
            ]}
           RedCurve {keys [
             <at 0 0 RCIM_Cubic>
             <at 0.1326 1>
             <at 0.2869 0>
             <at 0.4134 0.4394>
             <at 0.5507 5>
             <at 0.5846 0>
             <at 0.8175 0>
             {arriveTangent=31.5478, interpMode=RCIM_Cubic, leaveTangent=31.5478, time=0.8628, value=1}
             <at 0.976 5>
            ]}
          }}))
       Float=&FloatFromCurve(
        FloatCurve=DataInterface() {propertyValues {Curve {keys [{arriveTangent=-2.0684, interpMode=RCIM_Cubic, leaveTangent=-2.0684, tangentMode=RCTM_User, time=0, value=2}, <at 1 0>]}}}
        Scale Curve=&Multiply_Float(A=1, B=0.5)
       )
      )
     )
     "Float (Alpha)"=&Multiply_Float(
      A=&FloatFromCurve(
       FloatCurve=DataInterface() {propertyValues {Curve {keys [
           {arriveTangent=9.0725, interpMode=RCIM_Cubic, leaveTangent=9.0725, tangentMode=RCTM_User, time=0, value=0}
           {arriveTangent=-2.5249, interpMode=RCIM_Cubic, leaveTangent=-2.5249, tangentMode=RCTM_User, time=0.2, value=0.9847}
           {arriveTangent=-0.0405, interpMode=RCIM_Cubic, leaveTangent=-0.0405, tangentMode=RCTM_User, time=1, value=0}
          ]}}}
       Scale Curve=&RandomRangeFloat(Minimum=0.2, Maximum=1)
      )
      B=2
     )
    )
    Create Color Attribute=true
    Create Scale Attribute=true
    Module Output=::NewEnumerator1
    Scale=&Multiply_Vector(
     A=&RandomRangeVector(
      Minimum=<1 1 10>
      Maximum=&MakeCustomVectorFromBool(Bool=&RandomBool(), True Vector=<1.85 3 1>, False Vector=<1 1 6>)
     )
     B=&Multiply_VectorByFloat(
      Vector=<1 1 1>
      Float=&Multiply_Float(
       A=&MakeCustomFloatFromBool(Bool=&RandomBool(), True Float=0.1, False Float=0)
       B=&FloatFromCurve(
        FloatCurve=DataInterface() {propertyValues {Curve {keys [
            {arriveTangent=4.1785, time=0, value=1}
            <at 0.5 1.1>
            {arriveTangent=-0.0405, interpMode=RCIM_Cubic, leaveTangent=-0.0405, tangentMode=RCTM_User, time=1, value=1.5}
           ]}}}
        Scale Curve=&FloatFromCurve(
         FloatCurve=DataInterface() {propertyValues {Curve {keys [<at 0 1 RCIM_Constant>,<at 0.1 0 RCIM_Constant>,<at 0.4 1 RCIM_Constant>,<at 0.7 0 RCIM_Constant>,<at 1.2 1 RCIM_Constant>,<at 1.5 0 RCIM_Constant>,<at 3 1 RCIM_Constant>]}}}
         CurveIndex=$Emitter.LoopedAge
        )
       )
      )
     )
    )
   )
  ]
  rendererClasses [&NiagaraMeshRendererProperties]
  renderers [{rendererClass=&NiagaraMeshRendererProperties, rendererIndex=0}]
 }
 Glow_Trail {
  emitterUpdateScript [
   &EmitterState(Life Cycle Mode=::NewEnumerator1, Loop Duration=3, UseLoopDelay=false)
   &FN_SpawnAnimTrail(Number of Anim Trails=2)
  ]
  particleSpawnScript [
   &FN_InitializeAnimTrailBoneIndecies(Second Bone or Socket Index=2)
   &FN_InitializeAnimTrailBoneIndecies(AnimTrail Selection=1, First Bone or Socket Index=3, Second Bone or Socket Index=5) {module=FN_InitializeAnimTrailBoneIndecies001}
   &FN_InitializeAnimTrail(Lifetime=0.45)
   &AddVelocity(Velocity=&MakeCustomVectorFromBool(
     Bool=$User.bIsFrontEndPreview
     True Vector=<0 500 210>
     False Vector=&Lerp_Vector(
      A=&Lerp_Vector(A=<0 100 100>, B=<0 220 360>, Alpha=$System.InterpolateOverTime001.InterpolateOverTime_Float.Moving Average)
      B=<0 300 600>
      Alpha=$System.InterpolateOverTime.InterpolateOverTime_Float.Moving Average
     )
    ))
   &DynamicMaterialParameters(Module.Index0Float0Write=true, Module.Index0Float1Write=true, Module.Index0Float2Write=true, Module.Index0Float3Write=true)
  ]
  particleUpdateScript [
   &FN_UpdateAnimTrail()
   &SolveForcesAndVelocity(Clamp Velocity=false, Limit Acceleration=false)
   &ScaleRibbonWidth(Ribbon Width Scale=&MakeCustomFloatFromBool(
     Bool=$User.bIsFrontEndPreview
     True Float=0.9
     False Float=&Lerp_Float(
      A=&Lerp_Float(A=0.8, B=0.85, Alpha=$System.InterpolateOverTime001.InterpolateOverTime_Float.Moving Average)
      B=0.9
      Alpha=$System.InterpolateOverTime.InterpolateOverTime_Float.Moving Average
     )
    ))
   &ScaleColor(
    Scale Alpha=&Multiply_Float(
     A=&FloatFromCurve(
      FloatCurve=DataInterface() {propertyValues {Curve {keys [<at 0 1>,<at 0.2 1>,<at 0.6 0>,<at 1 0>]}}}
      Scale Curve=&Lerp_Float(A=1.5, B=0, Alpha=&Power(a=$System.InterpolateOverTime001.InterpolateOverTime_Float.Moving Average, b=0.3))
     )
     B=&MakeCustomFloatFromBool(
      Bool=$User.bIsFrontEndPreview
      True Float=&Multiply_Float(
       A=0.3
       B=&FloatFromCurve(FloatCurve=DataInterface() {propertyValues {Curve {keys [<at 0 0>,<at 1 0>,<at 1.3 1>]}}}, CurveIndex=$Emitter.Age)
      )
      False Float=&Multiply_Float(
       A=&SelectFloatFromArray(
        Direct Array Index=&MakeIntFromFloat(Float=$User.VariantSwitch)
        Float Selection Array=DataInterface() {propertyValues {FloatData [0.3,0.5,0.4,0.6]}}
       )
       B=&FloatFromCurve(FloatCurve=DataInterface() {propertyValues {Curve {keys [<at 0 0>,<at 1.7 0>,<at 2.2 1>]}}}, CurveIndex=$Emitter.Age)
      )
     )
    )
    Scale RGB=&MakeVectorFromLinearColorRGB(Color=&SelectLinearColorFromArray(
      Array Sampling Mode=::NewEnumerator1
      Direct Array Index=&MakeIntFromFloat(Float=$User.VariantSwitch)
      Color Selection Array=DataInterface() {propertyValues {ColorData [<rgba 1.5 0.15 0.3121 1>, <rgba 0.3616 0.15 1.5 1>, <rgba 1.5 0.3313 0 1>, <rgba 0.075 0.2461 1.5 1>]}}
      Random Seed=?
     ))
    ScaleA=true
    ScaleRGB=true
   )
   /SetVariables(Particles.Color002=&MakeLinearColorFromVectorAndFloat(
     "Vector (RGB)"=&MakeVectorFromLinearColorRGB(Color=&SelectLinearColorFromArray(
       Array Sampling Mode=::NewEnumerator1
       Direct Array Index=&MakeIntFromFloat(Float=$User.VariantSwitch)
       Color Selection Array=DataInterface() {propertyValues {ColorData [<rgba 1 0.15 0.2521 1>, <rgba 0.1905 0.1314 1.5 1>, <rgba 1 0.4199 0.08 1>, <rgba 0.0357 0.3612 1.5 1>]}}
       Random Seed=?
      ))
     "Float (Alpha)"=&Multiply_Float(
      A=&FloatFromCurve(
       FloatCurve=DataInterface() {propertyValues {Curve {keys [<at 0 1>,<at 0.2 1>,<at 0.6 0>,<at 1 0>]}}}
       Scale Curve=&Lerp_Float(A=1, B=0, Alpha=&Power(a=$System.InterpolateOverTime001.InterpolateOverTime_Float.Moving Average, b=0.3))
      )
      B=&MakeCustomFloatFromBool(
       Bool=$User.bIsFrontEndPreview
       True Float=&Multiply_Float(
        A=0.2
        B=&FloatFromCurve(FloatCurve=DataInterface() {propertyValues {Curve {keys [<at 0 0>,<at 1 0>,<at 1.3 1>]}}}, CurveIndex=$Emitter.Age)
       )
       False Float=&Multiply_Float(
        A=&SelectFloatFromArray(
         Direct Array Index=&MakeIntFromFloat(Float=$User.VariantSwitch)
         Float Selection Array=DataInterface() {propertyValues {FloatData [0.6,0.9,0.6,0.9]}}
        )
        B=&FloatFromCurve(FloatCurve=DataInterface() {propertyValues {Curve {keys [<at 0 0>,<at 1.7 0>,<at 2.2 1>]}}}, CurveIndex=$Emitter.Age)
       )
      )
     )
    )) {isSetParameters=true}
  ]
  rendererClasses [&NiagaraRibbonRendererProperties]
  renderers [
   {rendererClass=&NiagaraRibbonRendererProperties}
   {rendererClass=&NiagaraRibbonRendererProperties, rendererIndex=1}
  ]
 }
 Main_Trail {
  emitterUpdateScript [
   &EmitterState(Life Cycle Mode=::NewEnumerator1, Loop Delay=2, Loop Duration=3, UseLoopDelay=false)
   &FN_SpawnAnimTrail(Number of Anim Trails=2)
  ]
  particleSpawnScript [
   &FN_InitializeAnimTrailBoneIndecies(First Bone or Socket Index=1, Second Bone or Socket Index=0)
   &FN_InitializeAnimTrailBoneIndecies(AnimTrail Selection=1, First Bone or Socket Index=4, Second Bone or Socket Index=3) {module=FN_InitializeAnimTrailBoneIndecies001}
   &FN_InitializeAnimTrail()
   &AddVelocity(Velocity=&MakeCustomVectorFromBool(
     Bool=$User.bIsFrontEndPreview
     True Vector=<0 250 100>
     False Vector=&Lerp_Vector(
      A=&Lerp_Vector(A=<0 200 200>, B=<0 0 0>, Alpha=$System.InterpolateOverTime001.InterpolateOverTime_Float.Moving Average)
      B=<0 150 300>
      Alpha=$System.InterpolateOverTime.InterpolateOverTime_Float.Moving Average
     )
    ))
  ]
  particleUpdateScript [
   &FN_UpdateAnimTrail()
   &SolveForcesAndVelocity(Clamp Velocity=false, Limit Acceleration=false)
   &ScaleRibbonWidth(Ribbon Width Scale=&MakeCustomFloatFromBool(Bool=$User.bIsFrontEndPreview, True Float=0.9, False Float=&Lerp_Float(A=0, B=1, Alpha=$User.ForwardDot)))
   &ScaleColor(
    Scale Alpha=&MakeCustomFloatFromBool(
     Bool=$User.bIsFrontEndPreview
     True Float=&Multiply_Float(
      A=10
      B=&FloatFromCurve(
       FloatCurve=DataInterface() {propertyValues {Curve {keys [<at 0 0>,<at 1 0>,<at 1.3 1>]}}}
       CurveIndex=$Emitter.Age
       Scale Curve=&FloatFromCurve(
        FloatCurve=DataInterface() {propertyValues {Curve {keys [<at 0 0.2 RCIM_Constant>,<at 0.1 1 RCIM_Constant>,<at 0.4 0.2 RCIM_Constant>,<at 0.7 1 RCIM_Constant>,<at 1.2 0.2 RCIM_Constant>,<at 1.5 1 RCIM_Constant>,<at 3 0.2 RCIM_Constant>]}}}
        CurveIndex=$Emitter.LoopedAge
       )
      )
     )
     False Float=&Multiply_Float(
      A=10
      B=&Lerp_Float(
       A=-3
       B=&Multiply_Float(
        A=&FloatFromCurve(FloatCurve=DataInterface() {propertyValues {Curve {keys [<at 0 0>,<at 1.5 0>,<at 2 1>]}}}, CurveIndex=$Emitter.Age)
        B=&FloatFromCurve(
         FloatCurve=DataInterface() {propertyValues {Curve {keys [<at 0 0.2 RCIM_Constant>,<at 0.1 1 RCIM_Constant>,<at 0.4 0.2 RCIM_Constant>,<at 0.7 1 RCIM_Constant>,<at 1.2 0.2 RCIM_Constant>,<at 1.5 1 RCIM_Constant>,<at 3 0.2 RCIM_Constant>]}}}
         CurveIndex=$Emitter.LoopedAge
         Scale Curve=&SelectFloatFromArray(
          Direct Array Index=&MakeIntFromFloat(Float=$User.VariantSwitch)
          Float Selection Array=DataInterface() {propertyValues {FloatData [1,2,1.5,3]}}
         )
        )
       )
       Alpha=$System.InterpolateOverTime.InterpolateOverTime_Float.Moving Average
      )
     )
    )
    Scale RGB=&MakeVectorFromLinearColorRGB(Color=<rgba 0.5 0.5 0.5 1>)
    ScaleA=true
    ScaleRGB=true
   )
   /SetVariables(Particles.ColorPoint:LinearColor=&MakeLinearColorFromVectorAndFloat(
     "Vector (RGB)"=&Multiply_VectorByFloat(
      Vector=&MakeVectorFromLinearColorRGB(Color=&SelectLinearColorFromArray(
        Array Sampling Mode=::NewEnumerator1
        Direct Array Index=&MakeIntFromFloat(Float=$User.VariantSwitch)
        Color Selection Array=DataInterface() {propertyValues {ColorData [<rgba 5 4.1406 4.1406 1>, <rgba 5 4.1406 4.1406 1>, <rgba 5 4.1406 4.1406 1>, <rgba 4.8333 4.8554 8 1>]}}
        Random Seed=?
       ))
      Float=3
     )
     "Float (Alpha)"=&MakeCustomFloatFromBool(
      Bool=$User.bIsFrontEndPreview
      True Float=&Multiply_Float(
       A=1
       B=&FloatFromCurve(
        FloatCurve=DataInterface() {propertyValues {Curve {keys [<at 0 0>,<at 1 0>,<at 1.3 1>]}}}
        CurveIndex=$Emitter.Age
        Scale Curve=&FloatFromCurve(
         FloatCurve=DataInterface() {propertyValues {Curve {keys [<at 0 0.2 RCIM_Constant>,<at 0.1 1 RCIM_Constant>,<at 0.4 0.2 RCIM_Constant>,<at 0.7 1 RCIM_Constant>,<at 1.2 0.2 RCIM_Constant>,<at 1.5 1 RCIM_Constant>,<at 3 0.2 RCIM_Constant>]}}}
         CurveIndex=$Emitter.LoopedAge
        )
       )
      )
      False Float=&Multiply_Float(
       A=2
       B=&Lerp_Float(
        A=-4
        B=&Multiply_Float(
         A=&FloatFromCurve(FloatCurve=DataInterface() {propertyValues {Curve {keys [<at 0 0>,<at 1.5 0>,<at 2 1>]}}}, CurveIndex=$Emitter.Age)
         B=&FloatFromCurve(
          FloatCurve=DataInterface() {propertyValues {Curve {keys [<at 0 0.2 RCIM_Constant>,<at 0.1 1 RCIM_Constant>,<at 0.4 0.2 RCIM_Constant>,<at 0.7 1 RCIM_Constant>,<at 1.2 0.2 RCIM_Constant>,<at 1.5 1 RCIM_Constant>,<at 3 0.2 RCIM_Constant>]}}}
          CurveIndex=$Emitter.LoopedAge
         )
        )
        Alpha=$System.InterpolateOverTime.InterpolateOverTime_Float.Moving Average
       )
      )
     )
    )) {isSetParameters=true}
   /SetVariables(Particles.Color_Bottom02:LinearColor=&MakeLinearColorFromVectorAndFloat(
     "Vector (RGB)"=&MakeVectorFromLinearColorRGB(Color=<rgba 0.7 0.3865 0.3865 1>)
     "Float (Alpha)"=&MakeCustomFloatFromBool(
      Bool=$User.bIsFrontEndPreview
      True Float=&Multiply_Float(
       A=20
       B=&FloatFromCurve(
        FloatCurve=DataInterface() {propertyValues {Curve {keys [<at 0 0>,<at 1 0>,<at 1.3 1>]}}}
        CurveIndex=$Emitter.Age
        Scale Curve=&FloatFromCurve(
         FloatCurve=DataInterface() {propertyValues {Curve {keys [<at 0 0.2 RCIM_Constant>,<at 0.1 1 RCIM_Constant>,<at 0.4 0.2 RCIM_Constant>,<at 0.7 1 RCIM_Constant>,<at 1.2 0.2 RCIM_Constant>,<at 1.5 1 RCIM_Constant>,<at 3 0.2 RCIM_Constant>]}}}
         CurveIndex=$Emitter.LoopedAge
        )
       )
      )
      False Float=&Multiply_Float(
       A=30
       B=&Lerp_Float(
        A=-3
        B=&Multiply_Float(
         A=&FloatFromCurve(FloatCurve=DataInterface() {propertyValues {Curve {keys [<at 0 0>,<at 1.5 0>,<at 2 1>]}}}, CurveIndex=$Emitter.Age)
         B=&FloatFromCurve(
          FloatCurve=DataInterface() {propertyValues {Curve {keys [<at 0 0.2 RCIM_Constant>,<at 0.1 1 RCIM_Constant>,<at 0.4 0.2 RCIM_Constant>,<at 0.7 1 RCIM_Constant>,<at 1.2 0.2 RCIM_Constant>,<at 1.5 1 RCIM_Constant>,<at 3 0.2 RCIM_Constant>]}}}
          CurveIndex=$Emitter.LoopedAge
         )
        )
        Alpha=$System.InterpolateOverTime.InterpolateOverTime_Float.Moving Average
       )
      )
     )
    )) {isSetParameters=true}
   /SetVariables(Particles.Color_Rainbow:LinearColor=&MakeLinearColorFromVectorAndFloat(
     "Vector (RGB)"=&MakeVectorFromLinearColorRGB(Color=&SelectLinearColorFromArray(
       Array Sampling Mode=::NewEnumerator1
       Direct Array Index=&MakeIntFromFloat(Float=$User.VariantSwitch)
       Color Selection Array=DataInterface() {propertyValues {ColorData [<rgba 0.9365 1.068 1.2 1>, <rgba 0.4065 1.2 1.1541 1>, <rgba 1.2 1.2 1.2 1>, <rgba 0.4065 1.2 1.1541 1>]}}
       Random Seed=?
      ))
     "Float (Alpha)"=&MakeCustomFloatFromBool(
      Bool=$User.bIsFrontEndPreview
      True Float=&Multiply_Float(
       A=6
       B=&FloatFromCurve(
        FloatCurve=DataInterface() {propertyValues {Curve {keys [<at 0 0>,<at 1 0>,<at 1.3 1>]}}}
        CurveIndex=$Emitter.Age
        Scale Curve=&FloatFromCurve(
         FloatCurve=DataInterface() {propertyValues {Curve {keys [<at 0 1 RCIM_Constant>,<at 0.1 -0.2 RCIM_Constant>,<at 0.4 1 RCIM_Constant>,<at 0.7 -0.2 RCIM_Constant>,<at 1.2 1 RCIM_Constant>,<at 1.5 -0.2 RCIM_Constant>,<at 3 1 RCIM_Constant>]}}}
         CurveIndex=$Emitter.LoopedAge
        )
       )
      )
      False Float=&Multiply_Float(
       A=15
       B=&Lerp_Float(
        A=-3
        B=&Multiply_Float(
         A=&FloatFromCurve(FloatCurve=DataInterface() {propertyValues {Curve {keys [<at 0 0>,<at 1.5 0>,<at 2 1>]}}}, CurveIndex=$Emitter.Age)
         B=&FloatFromCurve(
          FloatCurve=DataInterface() {propertyValues {Curve {keys [<at 0 1 RCIM_Constant>,<at 0.1 0 RCIM_Constant>,<at 0.4 1 RCIM_Constant>,<at 0.7 0 RCIM_Constant>,<at 1.2 1 RCIM_Constant>,<at 1.5 0 RCIM_Constant>,<at 3 1 RCIM_Constant>]}}}
          CurveIndex=$Emitter.LoopedAge
         )
        )
        Alpha=$System.InterpolateOverTime.InterpolateOverTime_Float.Moving Average
       )
      )
     )
    )) {isSetParameters=true}
   /SetVariables(Particles.Color_Rainbow_Out:LinearColor=&MakeLinearColorFromVectorAndFloat(
     "Vector (RGB)"=&MakeVectorFromLinearColorRGB(Color=&SelectLinearColorFromArray(
       Array Sampling Mode=::NewEnumerator1
       Direct Array Index=&MakeIntFromFloat(Float=$User.VariantSwitch)
       Color Selection Array=DataInterface() {propertyValues {ColorData [<rgba 2 0.1042 0.1828 1>, <rgba 0.8995 0.09 3 1>, <rgba 2 0.3933 0 1>, <rgba 0 1.0478 2 1>]}}
       Random Seed=?
      ))
     "Float (Alpha)"=&MakeCustomFloatFromBool(
      Bool=$User.bIsFrontEndPreview
      True Float=&Multiply_Float(
       A=6
       B=&FloatFromCurve(
        FloatCurve=DataInterface() {propertyValues {Curve {keys [<at 0 0>,<at 1 0>,<at 1.3 1>]}}}
        CurveIndex=$Emitter.Age
        Scale Curve=&FloatFromCurve(
         FloatCurve=DataInterface() {propertyValues {Curve {keys [<at 0 1 RCIM_Constant>,<at 0.1 -0.2 RCIM_Constant>,<at 0.4 1 RCIM_Constant>,<at 0.7 -0.2 RCIM_Constant>,<at 1.2 1 RCIM_Constant>,<at 1.5 -0.2 RCIM_Constant>,<at 3 1 RCIM_Constant>]}}}
         CurveIndex=$Emitter.LoopedAge
        )
       )
      )
      False Float=&Multiply_Float(
       A=10
       B=&Lerp_Float(
        A=-3
        B=&Multiply_Float(
         A=&FloatFromCurve(FloatCurve=DataInterface() {propertyValues {Curve {keys [<at 0 0>,<at 1.5 0>,<at 2 1>]}}}, CurveIndex=$Emitter.Age)
         B=&FloatFromCurve(
          FloatCurve=DataInterface() {propertyValues {Curve {keys [<at 0 1 RCIM_Constant>,<at 0.1 0 RCIM_Constant>,<at 0.4 1 RCIM_Constant>,<at 0.7 0 RCIM_Constant>,<at 1.2 1 RCIM_Constant>,<at 1.5 0 RCIM_Constant>,<at 3 1 RCIM_Constant>]}}}
          CurveIndex=$Emitter.LoopedAge
         )
        )
        Alpha=$System.InterpolateOverTime.InterpolateOverTime_Float.Moving Average
       )
      )
     )
    )) {isSetParameters=true}
   &DynamicMaterialParameters(
    Index 2 Param 1=1
    Index 2 Param 3=1
    Index 2 Param 4=&SelectFloatFromArray(
     Direct Array Index=&MakeIntFromFloat(Float=$User.VariantSwitch)
     Float Selection Array=DataInterface() {propertyValues {FloatData [0,1,2,3]}}
    )
    Index 3 Param 1=&FloatFromCurve(
     FloatCurve=DataInterface() {propertyValues {Curve {keys [<at 0 1 RCIM_Constant>,<at 0.1 0 RCIM_Constant>,<at 0.4 1 RCIM_Constant>,<at 0.7 0 RCIM_Constant>,<at 1.2 1 RCIM_Constant>,<at 1.5 0 RCIM_Constant>,<at 3 1 RCIM_Constant>]}}}
     CurveIndex=$Emitter.LoopedAge
    )
    Module.Index2Float0Write=true
    Module.Index2Float1Write=true
    Module.Index2Float2Write=true
    Module.Index2Float3Write=true
    Module.Index3Float0Write=true
    Module.Index3Float1Write=false
    Module.Index3Float2Write=false
    Module.Index3Float3Write=false
    Write Parameter Index 0=::NewEnumerator0
    Write Parameter Index 2=::NewEnumerator1
    Write Parameter Index 3=::NewEnumerator1
   )
  ]
  rendererClasses [&NiagaraRibbonRendererProperties]
  renderers [
   {rendererClass=&NiagaraRibbonRendererProperties}
   {rendererClass=&NiagaraRibbonRendererProperties, rendererIndex=1}
   {rendererClass=&NiagaraRibbonRendererProperties, rendererIndex=2}
   {rendererClass=&NiagaraRibbonRendererProperties, rendererIndex=3}
   {rendererClass=&NiagaraRibbonRendererProperties, rendererIndex=4}
  ]
 }
 SpeedLine {
  simTarget=GPUComputeSim
  emitterUpdateScript [
   &EmitterState(
    Delay First Loop Only=true
    Inactive Response=::NewEnumerator1
    Life Cycle Mode=::NewEnumerator1
    Loop Delay=1
    Loop Duration=1
    UseLoopDelay=true
   )
   &SpawnRate(
    SpawnRate=&MakeCustomFloatFromBool(Bool=$User.bIsFrontEndPreview, True Float=200, False Float=150)
    Use Spawn Probability=true
   ) {module=SpawnRate001}
  ]
  particleSpawnScript [
   &InitializeParticle(
    Color=&MakeLinearColorFromVectorAndFloat(
     "Vector (RGB)"=<1 1 1>
     "Float (Alpha)"=&Multiply_Float(
      A=&UniformRangedFloat(Minimum=0.1, Maximum=0.2)
      B=&MakeCustomFloatFromBool(Bool=$User.bIsFrontEndPreview, True Float=0.5, False Float=1)
     )
    )
    Lifetime Max=0.1
    Lifetime Min=0.1
    Lifetime Mode=::NewEnumerator1
    Sprite Size Max=<2 15>
    Sprite Size Min=<2 5>
    Sprite Size Mode=::NewEnumerator2
    UsePositionOffset=false
    Write Material Random=true
   )
   &ShapeLocation(
    Handle Radius=0
    Large Radius=200
    MaskSpawnGroup=false
    Offset=&RandomRangeVector(Minimum=<0 0 0>, Maximum=<0 -300 0>)
    Offset Mode=::NewEnumerator0
    Override Seed=false
    Rotation Axis=<0 1 0>
    Rotation Mode=::NewEnumerator1
    Shape Origin=$Engine.Emitter.SimulationPosition
    Shape Primitive=::NewEnumerator3
    UseOwnerScale=true
   )
   &AddVelocity(Velocity=<0 0 500>) {module=AddVelocity001}
   &AddVelocity(Velocity=&Multiply_Vector(
     A=&RandomRangeVector(Minimum=<0 0 50>, Maximum=<0 0 200>)
     B=&MakeVector(X=1, Y=&MakeCustomFloatFromBool(Bool=$User.bIsFrontEndPreview, True Float=-1, False Float=1), Z=1)
    )) {enabled=false}
  ]
  particleUpdateScript [
   &ParticleState()
   &ScaleSpriteSize(Scale Factor=&Vector2DFromCurve(
     Vector2Curve=DataInterface() {propertyValues {XCurve {keys [<at 0 0.5 RCIM_Cubic>,<at 0.5 1 RCIM_Cubic>]}, YCurve {keys [<at 0 0 RCIM_Cubic>,<at 0.5 1 RCIM_Cubic>]}}}
     ScaleCurve=&Vector2DFromFloat(Value=1)
    ))
   &SolveForcesAndVelocity(Clamp Velocity=false, Limit Acceleration=false)
   &ScaleSpriteSizeBySpeed(Max Scale Factor=<1 30>, Min Scale Factor=<0.01 0.05>, Velocity Threshold=1)
   &ScaleColor(
    Scale Alpha=&FloatFromCurve(
     FloatCurve=DataInterface() {propertyValues {Curve {keys [
         <at 0 0 RCIM_Cubic>
         <at 0.2 1.0207>
         {arriveTangent=-2.221, interpMode=RCIM_Cubic, leaveTangent=-2.221, tangentMode=RCTM_User, time=1, value=0}
        ]}}}
     Scale Curve=&MakeCustomFloatFromBool(
      Bool=$User.bIsFrontEndPreview
      True Float=&MakeCustomFloatFromBool(Bool=$User.bIsFullyDeployed, True Float=0.5, False Float=0)
      False Float=&Lerp_Float(A=0, B=1, Alpha=$System.InterpolateOverTime.InterpolateOverTime_Float.Moving Average)
     )
    )
    Scale RGB=&VectorFromFloat(Value=3)
    ScaleA=true
    ScaleRGB=true
   )
  ]
  rendererClasses [&NiagaraSpriteRendererProperties]
  renderers [{rendererClass=&NiagaraSpriteRendererProperties}]
 }
 Trail_Sparkles {
  simTarget=GPUComputeSim
  emitterUpdateScript [
   &EmitterState(Life Cycle Mode=::NewEnumerator1, Loop Delay=1.15, Loop Duration=1, UseLoopDelay=false)
   &SpawnRate(
    Spawn Probability=&MakeCustomFloatFromBool(
     Bool=$User.bIsFrontEndPreview
     True Float=1
     False Float=&MakeCustomFloatFromBool(
      Bool=&SetBoolByFloatComparison(A=$User.ForwardDot, B=-0.2, Comparison Type=::NewEnumerator6)
      True Float=0
      False Float=1
     )
    )
    SpawnRate=35
    Use Spawn Probability=true
   )
  ]
  particleSpawnScript [
   &PartitionParticles()
   &InitializeParticle(
    Color=&SelectLinearColorFromArray(
     Array Sampling Mode=::NewEnumerator1
     Direct Array Index=&MakeIntFromFloat(Float=$User.VariantSwitch)
     Color Selection Array=DataInterface() {propertyValues {ColorData [<rgba 1 1 1 1>, <rgba 0.5848 0.2396 2 1>, <rgba 1.2 0.4146 0.0767 1>, <rgba 0.1007 0.5053 2 1>]}}
     Random Seed=?
    )
    Lifetime Max=0.2
    Lifetime Min=0.2
    Lifetime Mode=::NewEnumerator1
    Sprite Size Mode=::NewEnumerator1
    Uniform Sprite Size Min=10
    UsePositionOffset=false
    Write Material Random=true
   )
   &ShapeLocation(
    Cylinder Height=90
    Cylinder Radius=5
    MaskSpawnGroup=false
    Non Uniform Scale=<1 1 0.5>
    Offset Mode=::NewEnumerator0
    Override Seed=false
    Rotation Mode=::NewEnumerator1
    Shape Primitive=::NewEnumerator1
    Surface Only=false
    Surface Only Band Thickness=0.1
    UseOwnerScale=true
   )
  ]
  particleUpdateScript [
   &ParticleState()
   &SkeletalMeshLocation(
    Bone / Socket Index=&SelectIntFromArray(Direct Array Index=$Particles.PartitionParticles.Partition, Int Selection Array=DataInterface() {propertyValues {IntData [0,1]}})
    Bone Sampling Mode=::NewEnumerator4
    MaskSpawnGroup=false
    Sampled Position Offset=&AddVectorToPosition(
     Position=Linked() {linkedVariable {name=Particles.Initial.ShapeLocation.ShapeVector, type=Vector3f}}
     Vector=&MakeCustomVectorFromBool(
      Bool=$User.bIsFrontEndPreview
      True Vector=&Lerp_Vector(
       A=&SelectVectorFromArray(
        Direct Array Index=$Particles.PartitionParticles.Partition
        Vector Selection Array=DataInterface() {propertyValues {FloatData [<0 10 0>, <0 10 0>]}}
       )
       B=&SelectVectorFromArray(
        Direct Array Index=$Particles.PartitionParticles.Partition
        Vector Selection Array=DataInterface() {propertyValues {FloatData [<0 15 5>, <0 15 5>]}}
       )
       Alpha=&FloatFromCurve(FloatCurve=DataInterface() {propertyValues {Curve {keys [<at 0 0>,<at 1 1>]}}})
      )
      False Vector=&Lerp_Vector(
       A=&SelectVectorFromArray(
        Direct Array Index=$Particles.PartitionParticles.Partition
        Vector Selection Array=DataInterface() {propertyValues {FloatData [<0 8 5>, <0 8 5>]}}
       )
       B=&SelectVectorFromArray(
        Direct Array Index=$Particles.PartitionParticles.Partition
        Vector Selection Array=DataInterface() {propertyValues {FloatData [<0 10 10>, <0 10 10>]}}
       )
       Alpha=&FloatFromCurve(FloatCurve=DataInterface() {propertyValues {Curve {keys [<at 0 0>,<at 1 1>]}}})
      )
     )
    )
    Sampled Position Offset Coordinate Space:ENiagaraSkelMeshTransforms=::NewEnumerator2
    Skeletal Mesh=$System.Skeletal Mesh Reader
   )
   &CameraOffset(Camera Offset Amount=&MakeCustomFloatFromBool(
     Bool=$User.bIsFrontEndPreview
     True Float=0
     False Float=&Lerp_Float(A=0, B=-20, Alpha=$System.InterpolateOverTime.InterpolateOverTime_Float.Moving Average)
    ))
   &ScaleSpriteSize(Scale Factor=&Vector2DFromFloat(Value=&Lerp_Float(
      A=&FloatFromCurve(
       FloatCurve=DataInterface() {propertyValues {Curve {keys [<at 0 1>, {arriveTangent=-1.1583, leaveTangent=-1.1583, time=0.7, value=1}, {arriveTangent=-2.2869, time=1, value=0.4}]}}}
       Scale Curve=-1
      )
      B=&FloatFromCurve(
       FloatCurve=DataInterface() {propertyValues {Curve {keys [<at 0 1>, {arriveTangent=-1.1583, leaveTangent=-1.1583, time=0.9, value=1}, {arriveTangent=-2.2869, time=1, value=0}]}}}
       Scale Curve=-1
      )
      Alpha=$System.InterpolateOverTime.InterpolateOverTime_Float.Moving Average
     )))
   &ScaleColor(
    Scale Alpha=&FloatFromCurve(FloatCurve=DataInterface() {propertyValues {Curve {keys [<at 0 1>,<at 0.2 1>,<at 1 0>]}}}, Scale Curve=0.7)
    Scale RGB=&Multiply_Vector(
     A=/Niagara/DynamicInputs/Math/LerpMultipleVectors(
      Alpha:NiagaraFloat=$User.VariantSwitch
      V 1:Vector3f=&MakeVectorFromLinearColorRGB(Color=&ColorFromCurve(ColorCurve=DataInterface() {propertyValues {
          AlphaCurve {keys []}
          BlueCurve {keys [<at 0.1554 0.1398 RCIM_Cubic>,<at 1 1 RCIM_Cubic>]}
          GreenCurve {keys [<at 0.1554 0.1589 RCIM_Cubic>,<at 1 1 RCIM_Cubic>]}
          RedCurve {keys [<at 0.1554 1 RCIM_Cubic>,<at 1 1 RCIM_Cubic>]}
         }}))
      V 2:Vector3f=<1 1 1>
      V 3:Vector3f=<1 1 1>
      V 4:Vector3f=<1 1 1>
      V 5:Vector3f=?
      Vector Count:ENiagara_MultipleLerpCount=::NewEnumerator1
     )
     B=<1 1 1>
    )
    ScaleA=true
    ScaleRGB=true
   )
   /SetVariables(Particles.Glow03_Color=&MakeLinearColorFromVectorAndFloat(
     "Vector (RGB)"=&MakeVectorFromLinearColorRGB(Color=$Particles.Color)
     "Float (Alpha)"=&Multiply_Float(
      A=&FNTimeOfDayScalar(Night Scale=1, Morning Scale=0.3, Evening Scale=0.3)
      B=&FloatFromCurve(FloatCurve=DataInterface() {propertyValues {Curve {keys [<at 0 1>,<at 0.1 1>,<at 0.7 1>,<at 1 0>]}}}, Scale Curve=0.2)
     )
    )) {isSetParameters=true, module=SetVariables_#3}
   /SetVariables(Particles.Glow03_Size=&Multiply_Vector2D_ByFloat(
     Float=&Lerp_Float(
      A=&FloatFromCurve(
       FloatCurve=DataInterface() {propertyValues {Curve {keys [
           <at 0 0 RCIM_Cubic>
           <at 0.2 1.3 RCIM_Cubic>
           {arriveTangent=-0.1423, interpMode=RCIM_Cubic, leaveTangent=-0.1423, tangentMode=RCTM_User, time=1, value=0}
          ]}}}
       Scale Curve=20
      )
      B=&FloatFromCurve(
       FloatCurve=DataInterface() {propertyValues {Curve {keys [
           <at 0 0 RCIM_Cubic>
           {arriveTangent=-0.0069, interpMode=RCIM_Cubic, leaveTangent=-0.0069, tangentMode=RCTM_User, time=0.1, value=1.2}
           {arriveTangent=-0.6113, interpMode=RCIM_Cubic, leaveTangent=-0.6113, tangentMode=RCTM_User, time=0.7, value=1}
           {arriveTangent=-0.1423, interpMode=RCIM_Cubic, leaveTangent=-0.1423, tangentMode=RCTM_User, time=0.8, value=0}
          ]}}}
       Scale Curve=20
      )
      Alpha=$System.InterpolateOverTime.InterpolateOverTime_Float.Moving Average
     )
     Vector2D=<1 1>
    )) {isSetParameters=true, module=SetVariables_#4}
   &CurlNoiseForce(Bias Noise=false, Noise Strength=1000, Pan Noise=false)
   &SolveForcesAndVelocity(Clamp Velocity=false, Limit Acceleration=false)
  ]
  rendererClasses [&NiagaraSpriteRendererProperties]
  renderers [
   {rendererClass=&NiagaraSpriteRendererProperties}
   {rendererClass=&NiagaraSpriteRendererProperties, rendererIndex=1}
  ]
 }
 Trail_Sparkles_speed {
  simTarget=GPUComputeSim
  emitterUpdateScript [
   &EmitterState(Delay First Loop Only=true, Life Cycle Mode=::NewEnumerator1, Loop Delay=1.5, Loop Duration=3, UseLoopDelay=true)
   &SpawnRate(
    Spawn Probability=&MakeCustomFloatFromBool(
     Bool=$User.bIsFrontEndPreview
     True Float=0.8
     False Float=&Lerp_Float(A=-2, B=1, Alpha=$System.InterpolateOverTime.InterpolateOverTime_Float.Moving Average)
    )
    SpawnRate=40
    Use Spawn Probability=true
   )
  ]
  particleSpawnScript [
   &PartitionParticles()
   &InitializeParticle(
    Color=&MakeCustomLinearColorFromBool(
     Bool=&RandomBool()
     True Color=&MakeCustomLinearColorFromBool(
      Bool=&RandomBool()
      True Color=&MakeCustomLinearColorFromBool(
       Bool=&RandomBool()
       True Color=&SelectLinearColorFromArray(
        Array Sampling Mode=::NewEnumerator1
        Direct Array Index=&MakeIntFromFloat(Float=$User.VariantSwitch)
        Color Selection Array=DataInterface() {propertyValues {ColorData [<rgba 1 0.07 0 1>, <rgba 0.2159 0.0785 1 1>, <rgba 1.1 0.2144 0 1>, <rgba 0 0.2192 1 1>]}}
        Random Seed=?
       )
       False Color=<rgba 1 0.8281 0 1>
      )
      False Color=&MakeCustomLinearColorFromBool(Bool=&RandomBool(), True Color=<rgba 0 1 0.1767 1>, False Color=<rgba 0 1 0.8022 1>)
     )
     False Color=&MakeCustomLinearColorFromBool(
      Bool=&RandomBool()
      True Color=&MakeCustomLinearColorFromBool(Bool=&RandomBool(), True Color=<rgba 0 0.1339 1 1>, False Color=<rgba 0 1 0.8617 1>)
      False Color=&MakeCustomLinearColorFromBool(Bool=&RandomBool(), True Color=<rgba 0.5824 0 1 1>, False Color=<rgba 1 0.8281 0 1>)
     )
    )
    Lifetime Max=0.6
    Lifetime Min=0.5
    Lifetime Mode=::NewEnumerator1
    Sprite Size Mode=::NewEnumerator1
    Uniform Sprite Size Max=11
    Uniform Sprite Size Min=10
    UsePositionOffset=false
    Write Material Random=true
   )
   &ShapeLocation(
    Cylinder Radius=&MakeCustomFloatFromBool(Bool=$User.bIsFrontEndPreview, True Float=8, False Float=20)
    MaskSpawnGroup=false
    Non Uniform Scale=<1 1 0.5>
    Offset Mode=::NewEnumerator0
    Override Seed=false
    Rotation Mode=::NewEnumerator1
    Shape Primitive=::NewEnumerator1
    Surface Only=false
    Surface Only Band Thickness=0.1
    UseOwnerScale=true
   )
  ]
  particleUpdateScript [
   &ParticleState()
   &SkeletalMeshLocation(
    Bone / Socket Index=&SelectIntFromArray(Direct Array Index=$Particles.PartitionParticles.Partition, Int Selection Array=DataInterface() {propertyValues {IntData [0,1]}})
    Bone Sampling Mode=::NewEnumerator4
    MaskSpawnGroup=false
    Sampled Position Offset=&AddVectorToPosition(
     Position=Linked() {linkedVariable {name=Particles.Initial.ShapeLocation.ShapeVector, type=Vector3f}}
     Vector=&MakeCustomVectorFromBool(
      Bool=$User.bIsFrontEndPreview
      True Vector=&Lerp_Vector(
       A=&SelectVectorFromArray(
        Direct Array Index=$Particles.PartitionParticles.Partition
        Vector Selection Array=DataInterface() {propertyValues {FloatData [<0 10 10>, <0 10 10>]}}
       )
       B=&SelectVectorFromArray(
        Direct Array Index=$Particles.PartitionParticles.Partition
        Vector Selection Array=DataInterface() {propertyValues {FloatData [<0 100 40>, <0 100 40>]}}
       )
       Alpha=&FloatFromCurve(FloatCurve=DataInterface() {propertyValues {Curve {keys [{arriveTangent=0.5521, time=0, value=0}, {arriveTangent=1.1328, leaveTangent=1.1328, time=1, value=1}]}}})
      )
      False Vector=&Lerp_Vector(
       A=&SelectVectorFromArray(
        Direct Array Index=$Particles.PartitionParticles.Partition
        Vector Selection Array=DataInterface() {propertyValues {FloatData [<0 10 20>, <0 10 20>]}}
       )
       B=&SelectVectorFromArray(
        Direct Array Index=$Particles.PartitionParticles.Partition
        Vector Selection Array=DataInterface() {propertyValues {FloatData [<0 50 110>, <0 50 110>]}}
       )
       Alpha=&FloatFromCurve(FloatCurve=DataInterface() {propertyValues {Curve {keys [{arriveTangent=0.5521, time=0, value=0}, {arriveTangent=1.1328, leaveTangent=1.1328, time=1, value=1}]}}})
      )
     )
    )
    Sampled Position Offset Coordinate Space:ENiagaraSkelMeshTransforms=::NewEnumerator2
    Skeletal Mesh=$System.Skeletal Mesh Reader
   )
   &CameraOffset(Camera Offset Amount=&MakeCustomFloatFromBool(
     Bool=$User.bIsFrontEndPreview
     True Float=0
     False Float=&Lerp_Float(A=0, B=-25, Alpha=$System.InterpolateOverTime.InterpolateOverTime_Float.Moving Average)
    ))
   &ScaleSpriteSize(Scale Factor=&Vector2DFromFloat(Value=&Lerp_Float(
      A=&FloatFromCurve(
       FloatCurve=DataInterface() {propertyValues {Curve {keys [<at 0 1>, {arriveTangent=-1.1583, leaveTangent=-1.1583, time=0.7, value=1}, {arriveTangent=-2.2869, time=1, value=0.4}]}}}
       Scale Curve=-1
      )
      B=&FloatFromCurve(
       FloatCurve=DataInterface() {propertyValues {Curve {keys [<at 0 1>, {arriveTangent=-1.1583, time=0.9, value=1}, {arriveTangent=-2.2869, time=1, value=0}]}}}
       Scale Curve=-1
      )
      Alpha=$System.InterpolateOverTime.InterpolateOverTime_Float.Moving Average
     )))
   &ScaleColor(
    Scale Alpha=&FloatFromCurve(FloatCurve=DataInterface() {propertyValues {Curve {keys [<at 0 1>,<at 0.2 1>,<at 1 0>]}}}, Scale Curve=0.7)
    ScaleA=true
    ScaleRGB=true
   )
   /SetVariables(Particles.Glow03_Color=&MakeLinearColorFromVectorAndFloat(
     "Vector (RGB)"=&MakeVectorFromLinearColorRGB(Color=$Particles.Color)
     "Float (Alpha)"=&Multiply_Float(
      A=&FNTimeOfDayScalar(Night Scale=1, Morning Scale=0.3, Evening Scale=0.3)
      B=&FloatFromCurve(FloatCurve=DataInterface() {propertyValues {Curve {keys [<at 0 1>,<at 0.1 1>,<at 0.7 1>,<at 1 0>]}}}, Scale Curve=0.2)
     )
    )) {isSetParameters=true, module=SetVariables_#3}
   /SetVariables(Particles.Glow03_Size=&Multiply_Vector2D_ByFloat(
     Float=&Lerp_Float(
      A=&FloatFromCurve(
       FloatCurve=DataInterface() {propertyValues {Curve {keys [
           <at 0 0 RCIM_Cubic>
           <at 0.2 1.3 RCIM_Cubic>
           {arriveTangent=-0.1423, interpMode=RCIM_Cubic, leaveTangent=-0.1423, tangentMode=RCTM_User, time=1, value=0}
          ]}}}
       Scale Curve=20
      )
      B=&FloatFromCurve(
       FloatCurve=DataInterface() {propertyValues {Curve {keys [
           <at 0 0 RCIM_Cubic>
           {arriveTangent=-0.0069, interpMode=RCIM_Cubic, leaveTangent=-0.0069, tangentMode=RCTM_User, time=0.1, value=1.2}
           {arriveTangent=-0.6113, interpMode=RCIM_Cubic, leaveTangent=-0.6113, tangentMode=RCTM_User, time=0.7, value=1}
           {arriveTangent=-0.1423, interpMode=RCIM_Cubic, leaveTangent=-0.1423, tangentMode=RCTM_User, time=0.8, value=0}
          ]}}}
       Scale Curve=20
      )
      Alpha=$System.InterpolateOverTime.InterpolateOverTime_Float.Moving Average
     )
     Vector2D=<1 1>
    )) {isSetParameters=true, module=SetVariables_#4}
   &CurlNoiseForce(Bias Noise=false, Noise Strength=1000, Pan Noise=false)
   &SolveForcesAndVelocity(Clamp Velocity=false, Limit Acceleration=false)
  ]
  rendererClasses [&NiagaraSpriteRendererProperties]
  renderers [
   {rendererClass=&NiagaraSpriteRendererProperties}
   {rendererClass=&NiagaraSpriteRendererProperties, rendererIndex=1}
  ]
 }
}
summary {
 systemName=NS_Glider_BerryFlight_Idle
 emitters [
  {emitterName=Main_Trail, rendererClasses [&NiagaraRibbonRendererProperties]}
  {emitterName=Glow_Trail, rendererClasses [&NiagaraRibbonRendererProperties]}
  {emitterName=Trail_Sparkles, simTarget=GPUComputeSim, rendererClasses [&NiagaraSpriteRendererProperties]}
  {emitterName=SpeedLine, simTarget=GPUComputeSim, rendererClasses [&NiagaraSpriteRendererProperties]}
  {emitterName=Glitch, simTarget=GPUComputeSim, rendererClasses [&NiagaraMeshRendererProperties]}
  {emitterName=Trail_Sparkles_speed, simTarget=GPUComputeSim, rendererClasses [&NiagaraSpriteRendererProperties]}
 ]
 userVariables [
  User.bIsFrontEnd {type=NiagaraBool}
  User.bIsFrontEndPreview {type=NiagaraBool}
  User.bIsFullyDeployed {type=NiagaraBool}
  User.ForwardDot {type=NiagaraFloat, defaultValue=0, description=""}
  User.ParaGlideLeanAlpha {type=NiagaraFloat, defaultValue=0, description=""}
  User.RibbonIDOffset {type=NiagaraInt32}
  User.VariantSwitch {type=NiagaraFloat, defaultValue=0, description=""}
 ]
}
